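# Looping in-air sound cue for the Dance Grenade projectile, spatialized with projectile attenuation.
# Chiller grenade and pitched-up building grenade loops are layered with a dance music loop that an Enveloper fades in.
# A quality-level switch sends the first two slots to this full mix with boosted Doppler, and the last two to just the grenade loops without music or Doppler.
# A concurrency override prevents new instances once the limit is reached.
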SoundCue /Game/Athena/Sounds/DanceGrenade/DanceGrenade_InAirLoop_Cue
root=QualityLevel_0
nodes [
 {id=Doppler_0, children [Mixer_2], properties {dopplerIntensity=2}}
 {
  id=Enveloper_0
  children [WavePlayer_5]
  properties {
   pitchCurve {editorCurveData {keys [{time=0}]}}
   volumeCurve {editorCurveData {keys [{interpMode=RCIM_Cubic, tangentMode=RCTM_Break, time=0, value=0}, <at 0.35 1>]}}
  }
 }
 {id=Mixer_1, children [WavePlayer_7,Modulator_0], properties {inputVolume [1,0.65]}}
 {id=Mixer_2, children [Mixer_1,Enveloper_0], properties {inputVolume [1,0.35]}}
 {id=Mixer_3, children [WavePlayer_10, Modulator_2], properties {inputVolume [1,0.65]}}
 {id=Modulator_0, children [WavePlayer_0], properties {pitchMax=1.2, pitchMin=1.2}}
 {id=Modulator_2, children [WavePlayer_8], properties {pitchMax=1.2, pitchMin=1.2}}
 {id=QualityLevel_0, children [Doppler_0,Doppler_0,Mixer_3,Mixer_3]}
 {id=WavePlayer_0, properties {bLooping=true, soundWaveAssetPtr=/Game/Sounds/Creative/Gadgets/Building/Building_Grenade_InAirLoop01}}
 {id=WavePlayer_10, properties {bLooping=true, soundWaveAssetPtr=/Game/Sounds/Fort_Weapons/Grenades/Chiller/ChillGrenade_InAirLoop_01}}
 {id=WavePlayer_5, properties {bLooping=true, soundWaveAssetPtr=&/music_dancegrenade_loop_InAir}}
 {id=WavePlayer_7, properties {bLooping=true, soundWaveAssetPtr=&ChillGrenade_InAirLoop_01}}
 {id=WavePlayer_8, properties {bLooping=true, soundWaveAssetPtr=&Building_Grenade_InAirLoop01}}
]
properties {
 attenuationSettings=/Game/Sounds/Attenuation/Fort_Weapon_Projectile_InAir_Loop_Att
 bOverrideConcurrency=true
 maxDistance=4500
 soundClassObject=&Fort_WeaponRanged_Projectile_Effects
 volumeMultiplier=0.8
 attenuationOverrides {falloffDistance=5000, reverbWetLevelMax=0.7, reverbWetLevelMin=0.2, attenuationShapeExtents {x=1000}}
 concurrencyOverrides {resolutionRule=PreventNew}
}
